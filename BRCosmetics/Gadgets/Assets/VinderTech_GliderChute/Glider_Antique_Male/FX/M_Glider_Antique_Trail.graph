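# Translucent unlit, two-sided trail material for Niagara mesh particles on the male Antique glider.
# A panning smoke texture, distorted by inky noise, drives particle-colored emissive and opacity, with length/width masks, noisy dissolve and depth and camera fades.
# World position offset sums sine-and-noise waving, SplineThicken width and a clamped, tail-weighted pull from Niagara parameters.
# That pull builds its target from the X parameter alone, repeated across all three axes.
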Material /BRCosmetics/Gadgets/Assets/VinderTech_GliderChute/Glider_Antique_Male/FX/M_Glider_Antique_Trail
graph {
 properties {
  bUsedWithNiagaraMeshParticles=true
  blendMode=BLEND_Translucent
  pixelDepthOffsetMode=PDOM_Legacy
  shadingModel=MSM_Unlit
  twoSided=true
  usedShadingModels=Unlit
 }
 nodes [
  n1 = Constant()
  n2 = MaterialFunctionCall(Float2=TextureCoordinate()) {materialFunction=&BreakOutFloat2Components}
  n3 = DynamicParameter() {defaultValue=<rgba 0 0 0.5 1>, paramNames [OffsetRandom,UVTileRandom,Dissolve,Param4]}
  n4 = MaterialFunctionCall(X=Add(A=@n2.R, B=@n3.UVTileRandom), Y=@n2.G) {materialFunction=&MakeFloat2}
  n5 = Add(A=Time(), B=@n3.OffsetRandom)
  n6 = ComponentMask(None=TextureSample(UVs=Panner(Coordinate=@n4.Result, Time=@n5) {speedX=-1}) {samplerType=SAMPLERTYPE_Grayscale, texture=&LowResBlurredNoise}.RGB) {b=true, g=true}
  n7 = Multiply(A=Multiply(A=AppendVector(A=@n1, B=@n6), B=VertexNormalWS())) {constB=2}
  n8 = ComponentMask(None=TextureCoordinate()) {r=true}
  n9 = Add(
   A=Constant() {r=2}
   B=DynamicParameter() {defaultValue=<rgba 0 0 0.5 1>, paramNames [OffsetRandom,UVTileRandom,Dissolve,Param4]}.UVTileRandom
  )
  n10 = Multiply(A=@n8, B=Add(A=@n9, B=Constant() {r=0.4952}))
  n11 = Time()
  n12 = Constant() {r=0.2667}
  n13 = Constant() {r=-1.2}
  n14 = Add(
   A=Add(
    A=Multiply(A=@n10) {constB=0.5}
    B=DynamicParameter() {defaultValue=<rgba 0 0 0.5 1>, paramNames [OffsetRandom,UVTileRandom,Dissolve,Param4]}.OffsetRandom
   )
   B=Multiply(A=@n11, B=Add(A=@n12, B=@n13)) {constB=-1}
  )
  n15 = Add(
   A=Multiply(A=Multiply(A=@n8, B=@n9)) {constB=0.5}
   B=DynamicParameter() {defaultValue=<rgba 0 0 0.5 1>, paramNames [OffsetRandom,UVTileRandom,Dissolve,Param4]}.OffsetRandom
  )
  n16 = Multiply(A=@n11, B=@n13) {constB=-1}
  n17 = MaterialFunctionCall(X=Constant(), Y=Sine(None=Add(A=@n14, B=@n12) {constB=0.2667}), Z=Sine(None=Add(A=@n15, B=@n16))) {materialFunction=&MakeFloat3}
  n18 = TextureCoordinate() {uTiling=2}
  n19 = Multiply(A=LinearInterpolate(A=Constant3Vector() {constant=<rgba 0 0 1 0>}, B=Add(A=@n7, B=@n17.Result), Alpha=ComponentMask(None=@n18) {r=true})) {constB=10}
  n20 = DynamicParameter() {parameterIndex=2, paramNames [Sine, SplineWidthBase, SplineWidthTip, Param4]}
  n21 = DynamicParameter() {parameterIndex=2, paramNames [Sine, SplineWidthBase, SplineWidthTip, Param4]}
  n22 = Multiply(B=@n21.SplineWidthBase) {constA=30}
  n23 = Multiply(B=@n21.SplineWidthTip) {constA=10}
  n24 = StaticBool()
  n25 = DynamicParameter() {desc=Parameters set by Niagara, parameterIndex=1, paramNames [X,Y,Z,WPOIntensity]}
  n26 = Subtract(
   A=MaterialFunctionCall(X=@n25.X, Y=@n25.X, Z=@n25.X) {materialFunction=&MakeFloat3}.Result
   B=ObjectPositionWS()
  )
  n27 = Power(Base=ComponentMask(None=TextureCoordinate()) {r=true}) {constExponent=3}
  Add(
   A=Add(
    A=Multiply(A=@n19, B=@n20.Sine)
    B=MaterialFunctionCall(WidthBase=@n22, WidthTip=@n23, Expand U or V UV Channel=@n24) {materialFunction=&SplineThicken}.WorldPositionOffset
   )
   B=LinearInterpolate(B=Clamp(None=@n26) {maxDefault=200, minDefault=-200}, Alpha=Multiply(A=@n25.WPOIntensity, B=@n27))
  )
  n28 = MaterialFunctionCall(Float2=TextureCoordinate()) {materialFunction=&BreakOutFloat2Components}
  n29 = MaterialFunctionCall(
   X=Add(A=@n28.R, B=DynamicParameter() {defaultValue=<rgba 0 0 0.5 1>, paramNames [OffsetRandom,UVTileRandom,Dissolve,Param4]}.UVTileRandom)
   Y=@n28.G
  ) {materialFunction=&MakeFloat2}
  n30 = TextureSample(UVs=Panner(Coordinate=@n29.Result, Time=@n5) {speedX=-0.3}) {texture=&T_Inky_Smoke_Tile}
  n31 = ScalarParameter() {defaultValue=0.15, parameterName=Trail Noise}
  n32 = Multiply(A=@n29.Result, B=AppendVector(A=ScalarParameter() {defaultValue=1, parameterName=Trail U Tile}, B=Constant() {r=1}))
  n33 = AppendVector(A=ScalarParameter() {defaultValue=-1, parameterName=Trail U Speed}, B=Constant())
  n34 = TextureSample(UVs=Add(A=Multiply(A=@n30.R, B=@n31) {constB=0.15}, B=Panner(Coordinate=@n32, Speed=@n33) {speedX=-1})) {samplerType=SAMPLERTYPE_Grayscale, texture=&T_Smoke_Trail_Soft}
  n35 = ParticleColor()
  MaterialFunctionCall(Emissive=Multiply(A=@n34.R, B=@n35.RGB)) {materialFunction=&MF_FX_EmissiveAdjustments}
  n36 = OneMinus(None=ComponentMask(None=TextureCoordinate()) {r=true})
  n37 = MaterialFunctionCall(Float2=TextureCoordinate()) {materialFunction=&BreakOutFloat2Components}
  n38 = TextureSample(UVs=MaterialFunctionCall(X=@n37.G, Y=@n37.G) {materialFunction=&MakeFloat2}.Result) {texture=/Game/Effects/Fort_Effects/Textures/Ramps/T_Power_Source_Sphere_Ramp}
  n39 = MaterialFunctionCall(Float2=TextureCoordinate()) {materialFunction=&BreakOutFloat2Components}
  n40 = MaterialFunctionCall(
   X=Add(A=@n39.R, B=DynamicParameter() {defaultValue=<rgba 0 0 0.5 1>, paramNames [OffsetRandom,UVTileRandom,Dissolve,Param4]}.UVTileRandom)
   Y=@n39.G
  ) {materialFunction=&MakeFloat2}
  n41 = Multiply(
   A=TextureSample(UVs=Panner(Coordinate=@n40.Result, Time=@n5) {speedX=-0.5}) {texture=&T_Inky_Smoke_Tile}.R
   B=Constant() {r=0.8}
  )
  n42 = TextureCoordinate() {uTiling=1.5}
  n43 = OneMinus(None=Saturate(None=ComponentMask(None=@n42) {r=true}))
  n44 = Constant() {r=3}
  n45 = Multiply(A=Saturate(None=Add(A=@n41, B=@n43)), B=@n44)
  n46 = LinearInterpolate(A=@n44, Alpha=DynamicParameter() {defaultValue=<rgba 0 0 0.5 1>, paramNames [OffsetRandom,UVTileRandom,Dissolve,Param4]}.Dissolve) {constB=-1}
  n47 = Multiply(A=Multiply(A=@n34.R, B=Multiply(A=@n36, B=@n38.R)), B=Saturate(None=Subtract(A=@n45, B=@n46)))
  n48 = MaterialFunctionCall(Distance=Constant() {r=512}) {materialFunction=&Near_Camera_Fade}
  Multiply(
   A=DepthFade(Opacity=Multiply(A=@n47, B=@n48.Result), FadeDistance=ScalarParameter() {defaultValue=20, parameterName=DepthFade}) {fadeDistanceDefault=20}
   B=@n35.A
  )
 ]
}
